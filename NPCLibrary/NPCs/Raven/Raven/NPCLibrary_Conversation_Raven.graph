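# Conversation database for the Raven NPC, giving it brooding greeting lines and access to shared NPC dialogue.
# The entry point leads to one speech node with a default line and contextual lines for returning players, first meetings, bar holdings and two specific outfits.
# Contextual lines pair a controller requirement with a curve-driven priority.
# The speech node then links out to the shared NPC quests and basic services conversations by entry tag.
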ConversationDatabase /NPCLibrary/NPCs/Raven/Raven/NPCLibrary_Conversation_Raven
entry_points [{tag=Conversation.Entry.Raven, destinations [#1]}]
nodes [
 ConversationEntryPointNode_7 {
  id=#1
  class=ConversationEntryPointNode
  outputs [#2]
  properties {entryTag {tagName=Conversation.Entry.Raven}}
 }
 Speech_C_4 {
  id=#2
  class=Speech_C
  outputs [#3, #4]
  properties {
   nodeName=""
   generalConfig {
    defaultMessage="My tears are black, like the night sky. Or my soul."
    contextualMessages [
     {
      message=You're slinking back into the inky darkness for more? Good.
      contextRequirements [{requirement {class=FortControllerRequirement_CheckTimeSinceLastPlayedMatch}}]
      priority {curve {rowName=Tandem.Speech.Priority.WhileSincePlayed}}
     }
     {
      message=An ebony cloud of darkness follows me always. So... you should stand back a bit.
      contextRequirements [{requirement {class=FortControllerRequirement_IsFirstEverConversationWithNPC}}]
      priority {curve {rowName=Tandem.Speech.Priority.FirstEverInteraction}}
     }
     {
      message=Bars will never heal the black depths of my soul.
      contextRequirements [{requirement {class=FortControllerRequirement_HasBars}}]
      priority {curve {rowName=Tandem.Speech.Priority.HighTotalBars}}
     }
     {
      message=Some green guy keeps coming around here looking for you.
      contextRequirements [{requirement {
         class=FortControllerRequirement_HasCID
         properties {allowedCIds [/BRCosmetics/Athena/Items/Cosmetics/Characters/CID_A_036_Athena_Commando_F_Obsidian]}
        }}]
      priority {curve {rowName=Tandem.Speech.Priority.OutfitReaction}}
     }
     {
      message=Sigh... you're looking for someone else.
      contextRequirements [{requirement {
         class=FortControllerRequirement_HasCID
         properties {allowedCIds [/BRCosmetics/Athena/Items/Cosmetics/Characters/CID_A_092_Athena_Commando_M_Broccoli_PR297]}
        }}]
      priority {curve {rowName=Tandem.Speech.Priority.OutfitReaction}}
     }
    ]
   }
  }
 }
 ConversationLinkNode_0 {id=#4, class=ConversationLinkNode, properties {remoteEntryTag {tagName=Conversation.Entry.BasicServices}}}
 ConversationLinkNode_1 {id=#3, class=ConversationLinkNode, properties {remoteEntryTag {tagName=Conversation.Entry.NPCLibrary.Quests}}}
]
